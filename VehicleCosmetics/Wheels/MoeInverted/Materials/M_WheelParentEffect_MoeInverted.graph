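# Parent effect material for the MoeInverted cosmetic wheel, a lit skeletal-mesh material with animated, tintable rim and tire effects.
# Vertex color red blends rim and tire for both color and normals, with RGB mask textures separating zones in each.
# A shared infinity-offset function, driven by a slow sine, layers with panning gradients and noise under color parameters.
# The tire adds fresnel, a borealis-style noise tint and sparkles masked by its combined detail normal.
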Material /VehicleCosmetics/Wheels/MoeInverted/Materials/M_WheelParentEffect_MoeInverted
graph {
 properties {
  bAutomaticallySetUsageInEditor=false
  bUsedWithSkeletalMesh=true
  previewMesh=/VehicleCosmetics/Wheels/MoeInverted/Meshes/NEW/SK_Wheel_MoeInverted
  usedShadingModels=Default Lit
 }
 nodes [
  Constant()
  Constant()
  Constant() {r=1}
  n1 = TextureObjectParameter() {parameterName=Rim_M, samplerType=SAMPLERTYPE_LinearGrayscale, texture=/VehicleCosmetics/Wheels/MoeInverted/Textures/T_MoeInverted_Rim_MASK}
  n2 = Time()
  n3 = Constant() {r=0.1}
  n4 = Multiply(A=Sine(None=Multiply(A=@n2, B=@n3)), B=@n3)
  n5 = MaterialFunctionCall!(Texture=@n1, Offset=@n4, RotateUVs, GradientDarkness=Constant() {r=0.1}, GradientFalloff, UVs, Opt_LowQualityOffset, Opt_MediumQualityOffset) {materialFunction=/VehicleCosmetics/SharedMaterials/MaterialFunctions/MF_Infinity_OffsetFunction}
  n6 = TextureSampleParameter2D() {parameterName=RimRGB, samplerType=SAMPLERTYPE_LinearColor, texture=/VehicleCosmetics/Wheels/MoeInverted/Textures/T_MoeInverted_Rim_RGB}
  n7 = Multiply(A=Constant() {r=0.8}, B=@n6.R)
  n8 = Rotator(Time=Constant() {r=0.5}) {speed=3.14}
  n9 = ConstantBiasScale(None=TextureSample(UVs=Panner(Coordinate=@n8) {speedY=0.2}) {texture=/VehicleCosmetics/Wheels/_SharedTextures/FX/Vert_Linear01_D}.R) {bias=-0.5, scale=3}
  n10 = Add(A=Add(A=ConstantBiasScale(None=@n5.Result) {bias=0, scale=1}, B=Add(A=@n6.G, B=@n7)), B=Multiply(A=@n6.R, B=Clamp(None=@n9)))
  n11 = VectorParameter() {parameterName=CustomColorRIM, defaultValue=<rgba 0.144 0.442 1.3 1>}
  n12 = Panner(Coordinate=Rotator(Coordinate=TextureCoordinate() {vTiling=0.5}, Time=Constant() {r=1}) {speed=3.14}) {speedX=-0.17}
  n13 = Multiply(
   A=Constant() {r=4}
   B=Multiply(
    A=@n6.A
    B=TextureSample(UVs=@n12) {samplerType=SAMPLERTYPE_LinearColor, texture=/VehicleCosmetics/Wheels/_SharedTextures/FX/Universal_Noise_82_RGB}.G
   )
  )
  n14 = VectorParameter() {parameterName=AccentColor, defaultValue=<rgba 1 0.6 0.21 1>}
  n15 = MaterialFunctionCall!(Normal, Exponent=Constant() {r=2}) {materialFunction=/VehicleCosmetics/Wheels/Cola_Holographic/Materials/MF_Fresnel_Classic}
  n16 = TextureSampleParameter2D(UVs=Panner(Coordinate=TextureCoordinate() {vTiling=2}) {speedX=0.02, speedY=0.02}) {parameterName=Noise_69, samplerType=SAMPLERTYPE_LinearColor, texture=/VehicleCosmetics/Wheels/_SharedTextures/FX/Universal_Noise_69_RGB}
  n17 = VectorParameter() {parameterName=BorealisticColor, defaultValue=<rgba 0.0635 0.718 0.421 1>}
  n18 = TextureCoordinate()
  n19 = Constant() {r=0.5}
  n20 = TextureSampleParameter2D(UVs=Panner(Coordinate=Rotator(Coordinate=@n18, Time=@n19) {speed=3.14}) {constCoordinate=3, speedX=0.01, speedY=-0.05}) {parameterName=Noise_69, samplerType=SAMPLERTYPE_LinearColor, texture=&Universal_Noise_69_RGB}
  n21 = Clamp(None=ConstantBiasScale(None=@n20.R) {bias=0.1, scale=2})
  n22 = Multiply(A=Power(Base=OneMinus(None=@n15.Result), Exp=Constant() {r=4}), B=Multiply(A=@n16.R, B=Multiply(A=@n17.RGB, B=@n21)))
  n23 = TextureSampleParameter2D() {
   parameterName=TireTreadRGB
   samplerType=SAMPLERTYPE_LinearColor
   texture=/VehicleCosmetics/Wheels/MoeInverted/Textures/T_MoeInverted_Tire_RGB
  }
  n24 = Clamp(None=MaterialFunctionCall(Exponent=Constant() {r=3}) {materialFunction=&MF_Fresnel_Classic}.Result)
  n25 = Multiply(A=Constant() {r=3}, B=@n24)
  n26 = Constant() {r=1.5}
  n27 = TextureSample(UVs=Panner(Coordinate=Rotator(Time=@n26) {speed=3.14}) {speedY=0.35}) {texture=&Vert_Linear01_D}
  n28 = MaterialFunctionCall(
   Texture=TextureObject() {samplerType=SAMPLERTYPE_LinearColor, texture=/VehicleCosmetics/Wheels/MoeInverted/Textures/T_MoeInverted_Tire_M1}
   Offset=@n4
   GradientDarkness=Constant() {r=0.01}
  ) {materialFunction=&MF_Infinity_OffsetFunction}
  n29 = Add(
   A=LinearInterpolate(A=Constant(), B=Multiply(A=@n27.R, B=@n28.Result), Alpha=@n23.B)
   B=Multiply(A=Constant() {r=3}, B=Multiply(A=@n23.G, B=@n27.R))
  )
  n30 = Constant() {r=0.5}
  n31 = Multiply(
   A=LinearInterpolate(A=Add(A=@n25, B=@n29), B=Multiply(A=@n24, B=@n30), Alpha=@n23.R)
   B=VectorParameter() {parameterName=CustomColor, defaultValue=<rgba 0.144 0.142 1.3 1>}.RGB
  )
  n32 = TextureSampleParameter2D(UVs=Panner(Coordinate=TextureCoordinate() {uTiling=3, vTiling=5}) {speedX=0.04, speedY=0.05}) {parameterName=Noise_69, samplerType=SAMPLERTYPE_LinearColor, texture=&Universal_Noise_69_RGB}
  n33 = TextureSampleParameter2D() {parameterName=Tire_Normal, samplerType=SAMPLERTYPE_Normal, texture=/VehicleCosmetics/Wheels/MoeInverted/Textures/T_MoeInverted_Tire_N}
  n34 = TextureSample(UVs=Panner(Coordinate=TextureCoordinate() {uTiling=2}) {speedX=-0.01}) {samplerType=SAMPLERTYPE_Normal, texture=/VehicleCosmetics/Wheels/_SharedTextures/FX/Sparkle_N}
  n35 = Constant3Vector() {constant=<rgba 0.5 0.5 0 0>}
  n36 = TextureSampleParameter2D() {
   parameterName=TireTreadRGB
   samplerType=SAMPLERTYPE_LinearColor
   texture=&T_MoeInverted_Tire_RGB
  }
  n37 = Add(
   A=Multiply(A=@n33.RGB, B=Constant3Vector() {constant=<rgba 0.625 0.625 1 0>})
   B=Multiply(A=Multiply(A=@n34.RGB, B=@n35), B=Add(A=@n36.R, B=@n36.R))
  )
  n38 = ConstantBiasScale(None=ComponentMask(None=@n37) {g=true}) {bias=-0.2, scale=2}
  n39 = Multiply(A=Constant() {r=5}, B=Multiply(A=ConstantBiasScale(None=@n32.G) {bias=-0.5, scale=1}, B=Clamp(None=@n38)))
  n40 = VertexColor()
  LinearInterpolate(
   A=Add(A=Multiply(A=@n10, B=@n11.RGB), B=Multiply(A=@n13, B=@n14.RGB))
   B=Add(A=Multiply(A=@n22, B=@n23.R), B=Add(A=@n31, B=@n39))
   Alpha=@n40.R
  )
  n41 = TextureSampleParameter2D() {parameterName=RimNormal, samplerType=SAMPLERTYPE_Normal, texture=/VehicleCosmetics/Wheels/MoeInverted/Textures/T_MoeInverted_Rim_N}
  LinearInterpolate(A=@n41.RGB, B=@n37, Alpha=@n40.R)
  Multiply(A=@n28.Result)
  OneMinus(None=@n23.R)
 ]
}
